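; Patrol point provider for Pistol Canary NPC heroes that configures each hero it spawns and registers it with the heroes manager.
; On authority, BeginPlay binds to the spawn-result dispatcher and handles successful spawns.
; Each hero's AI controller gets a radial leash centered on the provider, with scalable-float inner and outer radii.
; RegisterHeroPawn maps six AISpawnerData type tags to hero-kind enum keys in the manager's HeroKindMap, with optional debug prints.
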(event EventBeginPlay
  (Networking|SwitchHasAuthority
    (:Authority
      (EventDispatchers|BindEventtoOnActorSpawnedResult (delegate Custom|OnActorSpawned)))))

(event Custom|OnActorSpawned
  (params (SpawnedActor :type Actor%20Object%20Reference) (bResult :type Boolean))
  (if bResult
    (bind _asbp_pawn_pistol_canary_heroes (Utilities|Casting|CastToBP_Pawn_PistolCanary_Heroes SpawnedActor))
    (Development|PrintString "@@@ OnActorSpawned [Pistol_Canary_NPCHeroes_PatrolPointProvider]: CALLED!" false (Variables|Debug|GetbDebugPointProvider))
    (Utilities|FlowControl|Sequence
      (:then_0
        (bind _asfort_athena_aibot_controller (Utilities|Casting|CastToFortAthenaAIBotController (Pawn|GetController _asbp_pawn_pistol_canary_heroes)))
        (Class|BPFLAICommonFunctions|SetRadialLeash _asfort_athena_aibot_controller (Transformation|GetActorLocation) (ScalableFloat|GetValueatLevel (Variables|Internal|GetSF_PointLeash_InnerRadius) 1.0) (ScalableFloat|GetValueatLevel (Variables|Internal|GetSF_PointLeash_OuterRadius) 1.0)))
      (:then_1
        (CallFunction|RegisterHeroPawn _asbp_pawn_pistol_canary_heroes)))))

(fn RegisterHeroPawn
  (params (HeroPawn :type BP%20Pawn%20Pistol%20Canary%20Heroes%20Object%20Reference))
  (bind _heropawn (Variables|GetHeroPawn))
  (bind _heropawn_2 (Variables|GetHeroPawn))
  (bind _heropawn_3 (Variables|GetHeroPawn))
  (bind _heropawn_4 (Variables|GetHeroPawn))
  (bind _heropawn_5 (Variables|GetHeroPawn))
  (bind _heropawn_6 (Variables|GetHeroPawn))
  (Development|PrintString "@@@@@@ RegisterHeroPawn [Pistol_Canary_NPCHeroes_PatrolPointProvider]: CALLED!" false (Variables|Debug|GetbDebugPointProvider))
  (if (GameplayTags|HasMatchingGameplayTag (Variables|GetHeroPawn) "(TagName=\"AISpawnerData.Type.Dev07\")")
    (Utilities|Map|Add (Variables|/PistolCanary/Gameplay/Managers/BP_PistolCanary_NPCHeroes_Manager.BP_PistolCanary_NPCHeroes_Manager_C|GetHeroKindMap (Variables|Tunable|GetNPCHeroesManager)) "NewEnumerator0" _heropawn_6)
    (Development|PrintString (Utilities|String|Append "@@@@@@@@@ RegisterHeroPawn [Pistol_Canary_NPCHeroes_PatrolPointProvider]: Register MistRaven as " (Utilities|GetDisplayName _heropawn_6) "!") false (Variables|Debug|GetbDebugPointProvider))
    (elif (GameplayTags|HasMatchingGameplayTag (Variables|GetHeroPawn) "(TagName=\"AISpawnerData.Type.Dev08\")")
      (Utilities|Map|Add (Variables|/PistolCanary/Gameplay/Managers/BP_PistolCanary_NPCHeroes_Manager.BP_PistolCanary_NPCHeroes_Manager_C|GetHeroKindMap (Variables|Tunable|GetNPCHeroesManager)) "NewEnumerator2" _heropawn_2)
      (Development|PrintString (Utilities|String|Append "@@@@@@@@@ RegisterHeroPawn [Pistol_Canary_NPCHeroes_PatrolPointProvider]: Register AshenMagnus as " (Utilities|GetDisplayName _heropawn_2) "!") false (Variables|Debug|GetbDebugPointProvider))
      (elif (GameplayTags|HasMatchingGameplayTag (Variables|GetHeroPawn) "(TagName=\"AISpawnerData.Type.Dev09\")")
        (Utilities|Map|Add (Variables|/PistolCanary/Gameplay/Managers/BP_PistolCanary_NPCHeroes_Manager.BP_PistolCanary_NPCHeroes_Manager_C|GetHeroKindMap (Variables|Tunable|GetNPCHeroesManager)) "NewEnumerator3" _heropawn_3)
        (Development|PrintString (Utilities|String|Append "@@@@@@@@@ RegisterHeroPawn [Pistol_Canary_NPCHeroes_PatrolPointProvider]: Register PolarityWinn as " (Utilities|GetDisplayName _heropawn_3) "!") false (Variables|Debug|GetbDebugPointProvider))
        (elif (GameplayTags|HasMatchingGameplayTag (Variables|GetHeroPawn) "(TagName=\"AISpawnerData.Type.Dev10\")")
          (Utilities|Map|Add (Variables|/PistolCanary/Gameplay/Managers/BP_PistolCanary_NPCHeroes_Manager.BP_PistolCanary_NPCHeroes_Manager_C|GetHeroKindMap (Variables|Tunable|GetNPCHeroesManager)) "NewEnumerator1" _heropawn_4)
          (Development|PrintString (Utilities|String|Append "@@@@@@@@@ RegisterHeroPawn [Pistol_Canary_NPCHeroes_PatrolPointProvider]: Register BentBaton as " (Utilities|GetDisplayName _heropawn_4) "!") false (Variables|Debug|GetbDebugPointProvider))
          (elif (GameplayTags|HasMatchingGameplayTag (Variables|GetHeroPawn) "(TagName=\"AISpawnerData.Type.Dev11\")")
            (Utilities|Map|Add (Variables|/PistolCanary/Gameplay/Managers/BP_PistolCanary_NPCHeroes_Manager.BP_PistolCanary_NPCHeroes_Manager_C|GetHeroKindMap (Variables|Tunable|GetNPCHeroesManager)) "NewEnumerator4" _heropawn_5)
            (Development|PrintString (Utilities|String|Append "@@@@@@@@@ RegisterHeroPawn [Pistol_Canary_NPCHeroes_PatrolPointProvider]: Register SpatialTravel as " (Utilities|GetDisplayName _heropawn_5) "!") false (Variables|Debug|GetbDebugPointProvider))
            (elif (GameplayTags|HasMatchingGameplayTag (Variables|GetHeroPawn) "(TagName=\"AISpawnerData.Type.Dev12\")")
              (Utilities|Map|Add (Variables|/PistolCanary/Gameplay/Managers/BP_PistolCanary_NPCHeroes_Manager.BP_PistolCanary_NPCHeroes_Manager_C|GetHeroKindMap (Variables|Tunable|GetNPCHeroesManager)) "NewEnumerator5" _heropawn)
              (Development|PrintString (Utilities|String|Append "@@@@@@@@@ RegisterHeroPawn [Pistol_Canary_NPCHeroes_PatrolPointProvider]: Register LazyLizz as " (Utilities|GetDisplayName _heropawn) "!") false (Variables|Debug|GetbDebugPointProvider)))))))))

(fn ConstructionScript)
